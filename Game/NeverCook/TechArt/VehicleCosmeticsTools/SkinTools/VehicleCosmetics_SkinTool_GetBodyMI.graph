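# Editor utility task in the vehicle cosmetics skin tools that locates body material instances to serve as parent materials.
# After a notification and short delay, it stops if cancelled, otherwise looks up each body package name in the Asset Registry.
# It casts each first result to MaterialInstanceConstant and adds it to ParentMaterials, printing an error and breaking the loop on failure.
# The GetNamesData function is an unimplemented stub.
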EditorUtilityBlueprint /Game/NeverCook/TechArt/VehicleCosmeticsTools/SkinTools/VehicleCosmetics_SkinTool_GetBodyMI
graphs [
 EventGraph {nodes [
   Event BeginExecution()
   > SetTaskNotificationText!(execute, self:Editor Utility Task Object Reference, Text:"Text (by ref)"=MakeLiteralText(Value="LOCTEXT(\"Locating Bodies MI\")")) -> (then)
   > PrintString(InString=Locating Bodies MI)
   > Delay!(execute, WorldContextObject:Object Reference, Duration=3) -> (then)
   > n1 = Branch(Condition=Get CancelPending() -> (CancelPending:bool))
   n2 = For Each Loop with Break!(
    Exec:Exec=@n1.else
    Array:Array of Names=Get BodyMatInstancePackageName() -> (BodyMatInstancePackageName:Array of Names)
    Break:Exec=@n5.then
   ) -> (LoopBody:Exec, Array Element:Name, Array Index:int, Completed:Exec)
   > n3 = GetAssetsByPackageName!(
    execute
    self:Asset Registry Interface=GetAssetRegistry()
    PackageName:Name=@n2
    bIncludeOnlyOnDiskAssets:bool=false
    bSkipARFilteredAssets:bool=true
   ) -> (then, OutAssetData:Array of Asset Data Structures, ReturnValue:bool)
   > n4 = Cast To MaterialInstanceConstant(Object=GetAsset(InAssetData="Get (a copy)"(Array:Array of Asset Data Structures=@n3) -> (Output:Asset Data Structure)))
   > Add(
    TargetArray:Array of Material Instance Constant Object References=Get ParentMaterials() -> (ParentMaterials:Array of Material Instance Constant Object References)
    NewItem:"Material Instance Constant Object Reference (by ref)"=@n4
   ) {is=title K2Node_CallArrayFunction}
   n5 = PrintString(execute=@n4.CastFailed, InString=Could not locate body parent material. Might be issue with naming conventions)
   Event CancelRequested() {is=title K2Node_Event}
   > PrintString(InString=Task Cancelled)
   > Set CancelPending(CancelPending:bool=true) -> (Output_Get:bool)
  ]}
 GetNamesData {nodes [
   GetNamesData!() -> (then, AssetName:Name, PackagePath:Name) {is=title K2Node_FunctionEntry}
   > Return Node!(execute, SkinMI_Name:String, MaterialsPath:String, BodyName:String, BodyMI_Name:String, BodyMI_PackageName:Name=None)
  ]}
]
variables [
 BodyMaterialsPath {type="array<string>", default []}
 BodyMatInstanceNames {type="array<string>", default []}
 BodyMatInstancePackageName {type="array<string>", default []}
 BodyNames {type="array<string>", default []}
 CancelPending {type=boolean}
 NumberOfAssets {type=integer}
 ParentMaterials {type="array<MaterialInstanceConstant>", default []}
 SelectedObjectsArray {type="array<Object>", default []}
 SkinMaskTextures {type="array<Texture>", default []}
 SkinMatInstanceNames {type="array<string>", default []}
 SkinTexturePackagePaths {type="array<string>", default []}
]
