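# Editor utility widget for vehicle cosmetics that batch-creates skin material instances from textures selected in the Content Browser.
# GetNamesData derives instance names, body material paths and colorized status from naming conventions (_RGB, _BC, folders).
# Each instance gets its body's parent material, then SkinMask if colorized, otherwise SkinBaseColor with UseBaseColorTexture enabled.
# A second button logs the derived data; editor-task and item-definition creation chains are unwired.
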EditorUtilityWidgetBlueprint /Game/NeverCook/TechArt/VehicleCosmeticsTools/SkinTools/VehicleCosmetics_SkinTools
graphs [
 EventGraph {nodes [
   n1 = GetSelectedAssets()
   > n2 = For Each Loop with Break!(Exec:Exec, Array:Array of Object References=@n1, Break:Exec=@n9.then) -> (LoopBody:Exec, Array Element:Object Reference, Array Index:int, Completed:Exec)
   Set SelectedObjectsArray(execute=@n2.Completed, SelectedObjectsArray:Array of Object References=@n1) -> (Output_Get:Array of Object References)
   n3 = Get SkinMaskTextures() -> (SkinMaskTextures:Array of Texture Object References)
   > n4 = Set NumberOfAssets(NumberOfAssets:int=int - int(A=Length(TargetArray:Array of Texture Object References=@n3))) -> (Output_Get:int=0)
   > n5 = Branch(Condition="integer < integer"(A=@n4))
   n6 = Construct Object from Class(
    execute=@n5.else
    Class=&/VehicleCosmetics_SkinTool_GetData.VehicleCosmetics_SkinTool_GetData_C
   ) -> (ReturnValue:Vehicle Cosmetics Skin Tool Get Data Object Reference)
   > n7 = Set TaskBuildData(TaskBuildData:Vehicle Cosmetics Skin Tool Get Data Object Reference=@n6) -> (Output_Get:Vehicle Cosmetics Skin Tool Get Data Object Reference)
   > RegisterAndExecuteTask!(
    execute
    self:Editor Utility Subsystem Object Reference=Get EditorUtilitySubsystem()
    NewTask:Editor Utility Task Object Reference=@n7
    OptionalParentTask:Editor Utility Task Object Reference
   ) -> (then)
   n8 = Cast To Texture!(execute=@n2.LoopBody, Object=@n2) -> (then, CastFailed, AsTexture:Texture Object Reference) {is=title K2Node_DynamicCast}
   > Add(TargetArray:Array of Texture Object References=Get SkinMaskTextures(), NewItem:"Texture Object Reference (by ref)"=@n8) {is=title K2Node_CallArrayFunction}
   n9 = PrintString(execute=@n8.CastFailed, InString=Incorrect asset type selected. Select only skin textures)
   "On Clicked (Button_CreateAll)"() {is=title K2Node_ComponentBoundEvent}
   > n10 = Sequence!(execute) -> (then_0, then_1, then_2:Exec, then_3:Exec)
   > ClearArray()
   > BuildDataFromSelected()
   GetBodyParentMI(execute=@n10.then_1)
   CreateMI(execute=@n10.then_2)
   "On Clicked (EditorUtilityButton_1)"() {is=title K2Node_ComponentBoundEvent}
   > DebugAssetData()
   GetBodyParentMI() {is=title K2Node_CustomEvent}
   > n11 = For Each Loop with Break(Array:Array of Names=Get BodyMatInstancePackageName() -> (BodyMatInstancePackageName:Array of Names), Break=@n14.then) -> (Array Element:Name)
   > n12 = GetAssetsByPackageName!(
    execute
    self:Asset Registry Interface=GetAssetRegistry()
    PackageName:Name=@n11
    bIncludeOnlyOnDiskAssets:bool=false
    bSkipARFilteredAssets:bool=true
   ) -> (then, OutAssetData:Array of Asset Data Structures, ReturnValue:bool)
   > n13 = Cast To MaterialInstanceConstant(Object=GetAsset(InAssetData="Get (a copy)"(Array:Array of Asset Data Structures=@n12) -> (Output:Asset Data Structure)))
   > Add(
    TargetArray:Array of Material Instance Constant Object References=Get ParentMaterials() -> (ParentMaterials:Array of Material Instance Constant Object References)
    NewItem:"Material Instance Constant Object Reference (by ref)"=@n13
   )
   n14 = PrintString(execute=@n13.CastFailed, InString=Could not locate body parent material. Might be issue with naming conventions)
   CreateMI() {is=title K2Node_CustomEvent}
   > n15 = For Loop(LastIndex=Get NumberOfAssets() -> (NumberOfAssets:int=0))
   > n16 = GetAssetTools()
   > n17 = CreateAsset(
    self=@n16
    AssetName="Get (a ref)"(Array:Array of Strings=Get SkinMatInstanceNames() -> (SkinMatInstanceNames:Array of Strings), Dimension 1=@n15) -> (Output:"String (by ref)")
    PackagePath="Get (a copy)"(Array:Array of Strings=Get SkinTexturePackagePaths() -> (SkinTexturePackagePaths:Array of Strings), Dimension 1=@n15) -> (Output:String)
    AssetClass=/Script/Engine.MaterialInstanceConstant
   )
   > n18 = Cast To MaterialInstanceConstant(Object=@n17)
   > SetMaterialInstanceParent!(
    execute
    Instance:Material Instance Constant Object Reference=@n18
    NewParent:Material Interface Object Reference="Get (a copy)"(Array:Array of Material Instance Constant Object References=Get ParentMaterials(), Dimension 1=@n15) -> (Output:Material Instance Constant Object Reference)
   ) -> (then)
   n19 = "Get (a copy)"(Array:Array of Booleans=Get IsColorizedMaterial() -> (IsColorizedMaterial:Array of Booleans), Dimension 1=@n15) -> (Output:bool)
   > n20 = Branch(Condition=@n19)
   n21 = "Get (a copy)"(Array:Array of Texture Object References=Get SkinMaskTextures(), Dimension 1=@n15) -> (Output:Texture Object Reference)
   > SetMaterialInstanceTextureParameterValue!(
    execute
    Instance:Material Instance Constant Object Reference=@n18
    ParameterName:Name=SkinMask
    Value:Texture Object Reference=@n21
    Association:EMaterialParameterAssociation Enum=GlobalParameter
   ) -> (then, ReturnValue:bool)
   SetMaterialInstanceStaticSwitchParameterValue!(
    execute=@n20.else
    Instance:Material Instance Constant Object Reference=@n18
    ParameterName:Name=UseBaseColorTexture
    Value:bool=InvertBool!(bValue:bool=@n19) -> (ReturnValue:bool)
    Association:EMaterialParameterAssociation Enum=GlobalParameter
    bUpdateMaterialInstance:bool=true
   ) -> (then, ReturnValue:bool)
   > SetMaterialInstanceTextureParameterValue(Instance=@n18, ParameterName=SkinBaseColor, Value=@n21, Association=GlobalParameter)
   BuildDataFromSelected() {is=title K2Node_CustomEvent}
   > n22 = GetSelectedAssets()
   > n23 = For Each Loop with Break(Array:Array of Object References=@n22, Break=@n32.then) -> (Array Element:Object Reference)
   n24 = Set SelectedObjectsArray(execute=@n23.Completed, SelectedObjectsArray=@n22)
   n25 = Get SkinMaskTextures()
   > n26 = Set NumberOfAssets(NumberOfAssets=int - int(A=Length(TargetArray=@n25))) -> (Output_Get=0)
   > n27 = Branch(Condition="integer < integer"(A=@n26))
   n28 = For Each Loop with Break(Exec=@n27.else, Array=@n24)
   n29 = CreateAssetData!(InAsset=@n28, bAllowBlueprintClass=false) -> (
    ReturnValue_PackageName:Name
    ReturnValue_PackagePath:Name
    ReturnValue_AssetName:Name
    ReturnValue_AssetClassPath:Top Level Asset Path Structure
   )
   > n30 = GetNamesData!(execute, self, AssetName:Name=@n29.ReturnValue_AssetName, PackagePath:Name=@n29.ReturnValue_PackagePath) -> (
    then
    SkinMI_Name:String
    MaterialsPath:String
    BodyName:String
    BodyMI_Name:String
    BodyMI_PackageName:Name=None
    ReturnValue:String
    isColorized:bool
   )
   > Add(TargetArray:Array of Strings=Get SkinMatInstanceNames(), NewItem:"String (by ref)"=@n30)
   > Add(TargetArray=Get BodyMaterialsPath() -> (BodyMaterialsPath:Array of Strings), NewItem=@n30.MaterialsPath)
   > Add(TargetArray=Get BodyNames() -> (BodyNames:Array of Strings), NewItem=@n30.BodyName)
   > Add(TargetArray=Get BodyMatInstanceNames() -> (BodyMatInstanceNames:Array of Strings), NewItem=@n30.BodyMI_Name)
   > Add(TargetArray:Array of Names=Get BodyMatInstancePackageName(), NewItem:"Name (by ref)"=@n30.BodyMI_PackageName)
   > Add(
    TargetArray:Array of Strings=Get SkinTexturePackagePaths()
    NewItem:"String (by ref)"="To String (Name)"(InName=@n29.ReturnValue_PackagePath)
   )
   > Add(TargetArray:Array of Booleans=Get IsColorizedMaterial(), NewItem:"Boolean (by ref)"=@n30.isColorized)
   n31 = Cast To Texture(execute=@n23.LoopBody, Object=@n23)
   > Add(TargetArray:Array of Texture Object References=Get SkinMaskTextures(), NewItem:"Texture Object Reference (by ref)"=@n31)
   n32 = PrintString(execute=@n31.CastFailed, InString=Incorrect asset type selected. Select only skin textures)
   DebugAssetData() {is=title K2Node_CustomEvent}
   > K2Node_Composite_4 = DebugData() {is=title K2Node_Composite}
   n33 = Construct Object from Class(Class=&/VehicleCosmetics_SkinTool_GetBodyMI.VehicleCosmetics_SkinTool_GetBodyMI_C) -> (ReturnValue:Vehicle Cosmetics Skin Tool Get Body MI Object Reference)
   > RegisterAndExecuteTask(
    self=Get EditorUtilitySubsystem()
    NewTask=@n33
    OptionalParentTask=Get TaskBuildData() -> (TaskBuildData:Vehicle Cosmetics Skin Tool Get Data Object Reference)
   )
   n34 = For Loop(LastIndex=Get NumberOfAssets() -> (NumberOfAssets=0))
   > n35 = GetAssetTools()
   n36 = "Get (a copy)"(Array:Array of Strings=Get SkinMatInstanceNames(), Dimension 1=@n34) -> (Output:String)
   n37 = Append(A=/VehicleCosmetics/Mutable/Skins/, B="Get (a copy)"(Array=Get BodyNames(), Dimension 1=@n34))
   > CreateAsset(
    self=@n35
    AssetName=Replace(SourceString=Replace(SourceString=@n36, From=MI_), From=Skin, To=CarSkin)
    PackagePath=@n37
    AssetClass=/Script/VehicleCosmeticsRuntime.FortVehicleCosmeticsItemDefinition_Skin
   )
   n38 = "Get (a copy)"(Array=Get SkinMatInstanceNames(), Dimension 1=@n34)
   > CreateAsset(
    self=@n35
    AssetName=Replace(SourceString=Replace(SourceString=@n38, From=MI_, To=VCID_), From=Skin, To=CarSkin)
    PackagePath=@n37
    AssetClass=/Script/VehicleCosmeticsRuntime.VehicleCosmeticsItemDefinition_Skin
   )
   K2Node_Composite_3 = ClearingArray!() {is=title K2Node_Composite}
   CreateIDs() {is=title K2Node_CustomEvent}
  ]}
 ClearingArray {
  path=EventGraph.K2Node_Composite_3.ClearingArray
  nodes [
   ClearArray() {is=title K2Node_CustomEvent}
   > Clear(TargetArray:Array of Object References=Get SelectedObjectsArray() -> (SelectedObjectsArray:Array of Object References))
   > Clear(TargetArray:Array of Strings=Get SkinMatInstanceNames())
   > Clear(TargetArray=Get BodyMaterialsPath())
   > Clear(TargetArray=Get BodyNames())
   > Clear(TargetArray=Get BodyMatInstanceNames())
   > Clear(TargetArray:Array of Names=Get BodyMatInstancePackageName())
   > Clear(TargetArray:Array of Material Instance Constant Object References=Get ParentMaterials())
   > Clear(TargetArray:Array of Strings=Get SkinTexturePackagePaths())
   > Clear(TargetArray:Array of Texture Object References=Get SkinMaskTextures())
   > Clear(TargetArray:Array of Booleans=Get IsColorizedMaterial())
   Inputs!()
   Outputs()
  ]
 }
 DebugData {
  path=EventGraph.K2Node_Composite_4.DebugData
  nodes [
   Inputs!() -> (execute:Exec)
   > ClearArray() {is=title K2Node_CallFunction}
   > BuildDataFromSelected() {is=title K2Node_CallFunction}
   > PrintDataFromArrayString!(execute, self, ArrayName:String=Skin Mat Instance Names, Array:Array of Strings=Get SkinMatInstanceNames()) -> (then)
   > PrintDataFromArrayString(ArrayName=Body Material Paths, Array=Get BodyMaterialsPath())
   > PrintDataFromArrayString(ArrayName=BodyNames, Array=Get BodyNames())
   > PrintDataFromArrayString(ArrayName=Body Material Instance Names, Array=Get BodyMatInstanceNames())
   > PrintDataFromArrayString(ArrayName=Skin texture package paths, Array=Get SkinTexturePackagePaths())
   > PrintString(InString=Append(A=Body Material Instance package names, B="-------------------\r\n"))
   > n1 = For Each Loop(Array:Array of Names=Get BodyMatInstancePackageName()) -> (Array Element:Name)
   PrintString(execute=@n1.Completed, InString=Append(A=Skin mask textures, B="-------------------\r\n"))
   > n2 = For Each Loop(Array:Array of Texture Object References=Get SkinMaskTextures()) -> (Array Element:Texture Object Reference)
   PrintString(execute=@n2.Completed, InString=Append(A=selected objects, B="-------------------\r\n"))
   > n3 = For Each Loop(Array:Array of Object References=Get SelectedObjectsArray()) -> (Array Element:Object Reference)
   PrintString(execute=@n3.Completed, InString=Append(A=Parent materials, B="-------------------\r\n"))
   > n4 = For Each Loop(Array:Array of Material Instance Constant Object References=Get ParentMaterials()) -> (Array Element:Material Instance Constant Object Reference)
   > PrintString(InString=Append!(A="To String (Integer)"(InInt=@n4.Array Index), B=": ", C:String=GetDisplayName(Object=@n4)) -> (ReturnValue))
   PrintString(execute=@n3.LoopBody, InString=Append(A="To String (Integer)"(InInt=@n3.Array Index), B=": ", C=GetDisplayName(Object=@n3)))
   PrintString(execute=@n2.LoopBody, InString=Append(A="To String (Integer)"(InInt=@n2.Array Index), B=": ", C=GetDisplayName(Object=@n2)))
   PrintString(execute=@n1.LoopBody, InString=Append(A="To String (Integer)"(InInt=@n1.Array Index), B=": ", C="To String (Name)"(InName=@n1)))
   Outputs()
  ]
 }
 GetNamesData {nodes [
   n1 = GetNamesData!() -> (then, AssetName:Name, PackagePath:Name) {is=title K2Node_FunctionEntry}
   n2 = "To String (Name)"(InName=@n1)
   > n3 = Branch(Condition=Contains(SearchIn=@n2, Substring=_RGB))
   > Set Skin_MI(Skin_MI:String=Replace(SourceString=@n2, From=_RGB, SearchCase=CaseSensitive)) -> (Output_Get:String)
   > n4 = Set isRGB(isRGB:bool=true) -> (Output_Get:bool)
   n5 = Replace(SourceString="To String (Name)"(InName=@n1.PackagePath), From=/Skins, To=/Materials, SearchCase=CaseSensitive)
   n6 = Replace(
    SourceString=Replace(SourceString="To String (Name)"(InName=@n1.PackagePath), From=/VehicleCosmetics/Bodies/, SearchCase=CaseSensitive)
    From=/Skins
    SearchCase=CaseSensitive
   )
   n7 = Append(A=MI_, B=@n6, C=_Body)
   Return Node!(
    execute=[@n8.then, @n4.then]
    SkinMI_Name:String=Append!(A=MI_, B=Get Skin_MI() -> (Skin_MI:String)) -> (ReturnValue)
    MaterialsPath:String=@n5
    BodyName:String=@n6
    BodyMI_Name:String=@n7
    BodyMI_PackageName:Name=String To Name(InString=Append!(A=@n5, B=/, C:String=@n7) -> (ReturnValue))
    ReturnValue:String
    isColorized:bool=Get isRGB() -> (isRGB:bool)
   )
   Set Skin_MI(execute=@n3.else, Skin_MI=Replace(SourceString=@n2, From=_BC, SearchCase=CaseSensitive))
   > n8 = Set isRGB(isRGB=false)
  ]}
 PrintDataFromArrayString {nodes [
   n1 = PrintDataFromArrayString!() -> (then, ArrayName:String, Array:Array of Strings) {is=title K2Node_FunctionEntry}
   > PrintString(InString=Append!(A=@n1, B="-------------------\r\n") -> (ReturnValue))
   > n2 = For Each Loop(Array:Array of Strings=@n1.Array) -> (Array Element:String)
   > PrintString(InString=Append!(A="To String (Integer)"(InInt=@n2.Array Index), B=": ", C:String=@n2) -> (ReturnValue))
  ]}
]
variables [
 BodyMaterialsPath {type="array<string>", default []}
 BodyMatInstanceNames {type="array<string>", default []}
 BodyMatInstancePackageName {type="array<string>", default []}
 BodyNames {type="array<string>", default []}
 IsColorizedMaterial {type="array<boolean>", default []}
 NumberOfAssets {type=integer}
 ParentMaterials {type="array<MaterialInstanceConstant>", default []}
 SelectedObjectsArray {type="array<Object>", default []}
 SkinMaskTextures {type="array<Texture>", default []}
 SkinMatInstanceNames {type="array<string>", default []}
 SkinTexturePackagePaths {type="array<string>", default []}
 TaskBuildData {type=VehicleCosmetics_SkinTool_GetData_C, default=null}
]
widgets {
 info {rootWidgetClass=&StackBox, widgetCount=17}
 widgets [
  StackBox_0 {widgetClassPath=&StackBox, widget=StackBox_0}
  VerticalBox_28 {
   widgetClassPath=&VerticalBox
   parent=StackBox_0
   widget=VerticalBox_28
   slot {
    class=StackBoxSlot
    properties {horizontalAlignment=HAlign_Left, verticalAlignment=VAlign_Top, padding=<ltrb 12 12 12 12>, size {sizeRule=Fill, value=1}}
   }
  }
  ScaleBox_7 {
   widgetClassPath=&ScaleBox
   parent=VerticalBox_28
   widget=ScaleBox_7
   slot {class=VerticalBoxSlot, properties {horizontalAlignment=HAlign_Center, padding=<ltrb 16 16 16 0>}}
  }
  VerticalBox_TopLogo {widgetClassPath=&VerticalBox, parent=ScaleBox_7, widget=VerticalBox_TopLogo, slot {class=ScaleBoxSlot}}
  HorizontalBox_13 {widgetClassPath=&HorizontalBox, parent=VerticalBox_TopLogo, widget=HorizontalBox_13, slot {class=VerticalBoxSlot}}
  SizeBox_701 {widgetClassPath=&SizeBox, parent=HorizontalBox_13, widget=SizeBox_701, slot {class=HorizontalBoxSlot}}
  Image_35 {
   widgetClassPath=&Image
   bIsVariable=true
   parent=SizeBox_701
   widget=Image_35
   properties {
    brush {
     bIsDynamicallyLoaded=false
     drawAs=Image
     imageType=NoImage
     mirroring=NoMirror
     resourceName=None
     resourceObject=/Game/NeverCook/TechArt/VehicleCosmeticsTools/VC_LOD_TOOL/Icons/UI_RL_Icon1
     tiling=NoTile
     imageSize=<64 64>
     margin=<ltrb 0 0 0 0>
     outlineSettings {
      bUseBrushTransparency=false
      roundingType=HalfHeightRadius
      width=0
      color {colorUseRule=UseColor_Specified, specifiedColor=<rgba 0 0 0 0>}
      cornerRadii=<0 0 0 0>
     }
     tintColor {colorUseRule=UseColor_Specified, specifiedColor=<rgba 1 1 1 1>}
     uVRegion {bIsValid=false, max=<0 0>, min=<0 0>}
    }
    colorAndOpacity=<rgba 0.665 0.665 0.665 1>
   }
   slot {class=SizeBoxSlot, properties {horizontalAlignment=HAlign_Left, verticalAlignment=VAlign_Center}}
  }
  SizeBox_754 {widgetClassPath=&SizeBox, parent=HorizontalBox_13, widget=SizeBox_754, slot {class=HorizontalBoxSlot}}
  TextBlock_310 {
   widgetClassPath=&TextBlock
   parent=SizeBox_754
   widget=TextBlock_310
   properties {
    text=Vehicle Cosmetics Decal Tool
    colorAndOpacity {colorUseRule=UseColor_Specified, specifiedColor=<rgba 0.599 0.599 0.599 1>}
    font {
     bForceMonospaced=false
     bMaterialIsStencil=false
     fontMaterial=null
     fontObject=null
     letterSpacing=0
     monospacedWidth=1
     size=18
     skewAmount=0
     typefaceFontName=Default
     outlineSettings {
      bApplyOutlineToDropShadows=false
      bMiteredCorners=false
      bSeparateFillAlpha=false
      outlineMaterial=null
      outlineSize=0
      outlineColor=<rgba 0 0 0 1>
     }
    }
   }
   slot {class=SizeBoxSlot, properties {padding=<ltrb 16 16 16 16>}}
  }
  ScaleBox_0 {
   widgetClassPath=&ScaleBox
   parent=VerticalBox_28
   widget=ScaleBox_0
   slot {class=VerticalBoxSlot, properties {padding=<ltrb 2 6 2 6>}}
  }
  Border_84 {
   widgetClassPath=&Border
   parent=ScaleBox_0
   widget=Border_84
   properties {
    background {
     bIsDynamicallyLoaded=false
     drawAs=Image
     imageType=NoImage
     mirroring=NoMirror
     resourceName=None
     resourceObject=null
     tiling=NoTile
     imageSize=<32 32>
     margin=<ltrb 0 0 0 0>
     outlineSettings {
      bUseBrushTransparency=false
      roundingType=HalfHeightRadius
      width=0
      color {colorUseRule=UseColor_Specified, specifiedColor=<rgba 0 0 0 0>}
      cornerRadii=<0 0 0 0>
     }
     tintColor {colorUseRule=UseColor_Specified, specifiedColor=<rgba 0.0176 0.0176 0.0176 1>}
     uVRegion {bIsValid=false, max=<0 0>, min=<0 0>}
    }
    desiredSizeScale=<1 5>
   }
   slot {class=ScaleBoxSlot, properties {horizontalAlignment=HAlign_Fill, verticalAlignment=VAlign_Fill}}
  }
  ScaleBox_2 {
   widgetClassPath=&ScaleBox
   parent=VerticalBox_28
   widget=ScaleBox_2
   slot {class=VerticalBoxSlot, properties {padding=<ltrb 2 2 2 24>}}
  }
  Button_CreateAll {
   widgetClassPath=&EditorUtilityButton
   parent=ScaleBox_2
   widget=Button_CreateAll
   properties {
    toolTipText="Select you skin mask texture(s) in the content browser before clicking."
    widgetStyle {
     clickedSlateSound {resourceObject=null}
     disabled {
      bIsDynamicallyLoaded=false
      drawAs=RoundedBox
      imageType=NoImage
      mirroring=NoMirror
      resourceName=None
      resourceObject=null
      tiling=NoTile
      imageSize=<0 0>
      margin=<ltrb 0 0 0 0>
      outlineSettings {
       bUseBrushTransparency=false
       roundingType=FixedRadius
       width=1
       color {colorUseRule=UseColor_ColorTable, specifiedColor=<rgba 0 0 0 0>}
       cornerRadii=<4 4 4 4>
      }
      tintColor {colorUseRule=UseColor_Specified, specifiedColor=<rgba 0.0395 0.0395 0.0395 1>}
      uVRegion {bIsValid=false, max=<0 0>, min=<0 0>}
     }
     disabledForeground {colorUseRule=UseColor_ColorTable, specifiedColor=<rgba 0 0 0 0>}
     hovered {
      bIsDynamicallyLoaded=false
      drawAs=RoundedBox
      imageType=NoImage
      mirroring=NoMirror
      resourceName=None
      resourceObject=null
      tiling=NoTile
      imageSize=<32 32>
      margin=<ltrb 0 0 0 0>
      outlineSettings {
       bUseBrushTransparency=false
       roundingType=FixedRadius
       width=1
       color {colorUseRule=UseColor_ColorTable, specifiedColor=<rgba 0 0 0 0>}
       cornerRadii=<4 4 4 4>
      }
      tintColor {colorUseRule=UseColor_Specified, specifiedColor=<rgba 0.9115 0.3654 0.1472 1>}
      uVRegion {bIsValid=false, max=<0 0>, min=<0 0>}
     }
     hoveredForeground {colorUseRule=UseColor_ColorTable, specifiedColor=<rgba 0 0 0 0>}
     hoveredSlateSound {resourceObject=null}
     normal {
      bIsDynamicallyLoaded=false
      drawAs=RoundedBox
      imageType=NoImage
      mirroring=NoMirror
      resourceName=None
      resourceObject=null
      tiling=NoTile
      imageSize=<32 32>
      margin=<ltrb 0 0 0 0>
      outlineSettings {
       bUseBrushTransparency=false
       roundingType=FixedRadius
       width=1
       color {colorUseRule=UseColor_ColorTable, specifiedColor=<rgba 0 0 0 0>}
       cornerRadii=<4 4 4 4>
      }
      tintColor {colorUseRule=UseColor_Specified, specifiedColor=<rgba 0.4167 0.119 0 1>}
      uVRegion {bIsValid=false, max=<0 0>, min=<0 0>}
     }
     normalForeground {colorUseRule=UseColor_ColorTable, specifiedColor=<rgba 0 0 0 0>}
     normalPadding=<ltrb 12 1.5 12 1.5>
     pressed {
      bIsDynamicallyLoaded=false
      drawAs=RoundedBox
      imageType=NoImage
      mirroring=NoMirror
      resourceName=None
      resourceObject=null
      tiling=NoTile
      imageSize=<32 32>
      margin=<ltrb 0 0 0 0>
      outlineSettings {
       bUseBrushTransparency=false
       roundingType=FixedRadius
       width=1
       color {colorUseRule=UseColor_ColorTable, specifiedColor=<rgba 0 0 0 0>}
       cornerRadii=<4 4 4 4>
      }
      tintColor {colorUseRule=UseColor_Specified, specifiedColor=<rgba 0.1302 0.0372 0 1>}
      uVRegion {bIsValid=false, max=<0 0>, min=<0 0>}
     }
     pressedForeground {colorUseRule=UseColor_ColorTable, specifiedColor=<rgba 0 0 0 0>}
     pressedPadding=<ltrb 12 2.5 12 0.5>
     pressedSlateSound {resourceObject=null}
    }
   }
   slot {class=ScaleBoxSlot, properties {horizontalAlignment=HAlign_Fill, verticalAlignment=VAlign_Fill}}
  }
  TextBlock_1 {
   widgetClassPath=&TextBlock
   parent=Button_CreateAll
   widget=TextBlock_1
   properties {
    text=Create Mat Instances
    colorAndOpacity {colorUseRule=UseColor_Specified, specifiedColor=<rgba 0.599 0.599 0.599 1>}
    font {
     bForceMonospaced=false
     bMaterialIsStencil=false
     fontMaterial=null
     fontObject=&Roboto
     letterSpacing=0
     monospacedWidth=1
     size=9
     skewAmount=0
     typefaceFontName=Bold
     outlineSettings {
      bApplyOutlineToDropShadows=false
      bMiteredCorners=false
      bSeparateFillAlpha=false
      outlineMaterial=null
      outlineSize=0
      outlineColor=<rgba 0 0 0 1>
     }
    }
   }
   slot {class=ButtonSlot}
  }
  ScaleBox_4 {
   widgetClassPath=&ScaleBox
   parent=VerticalBox_28
   widget=ScaleBox_4
   slot {class=VerticalBoxSlot, properties {padding=<ltrb 2 2 2 2>}}
  }
  EditorUtilityButton_1 {
   widgetClassPath=&EditorUtilityButton
   parent=ScaleBox_4
   widget=EditorUtilityButton_1
   properties {widgetStyle {
     clickedSlateSound {resourceObject=null}
     disabled {
      bIsDynamicallyLoaded=false
      drawAs=RoundedBox
      imageType=NoImage
      mirroring=NoMirror
      resourceName=None
      resourceObject=null
      tiling=NoTile
      imageSize=<0 0>
      margin=<ltrb 0 0 0 0>
      outlineSettings {
       bUseBrushTransparency=false
       roundingType=FixedRadius
       width=1
       color {colorUseRule=UseColor_ColorTable, specifiedColor=<rgba 0 0 0 0>}
       cornerRadii=<4 4 4 4>
      }
      tintColor {colorUseRule=UseColor_Specified, specifiedColor=<rgba 0.0395 0.0395 0.0395 1>}
      uVRegion {bIsValid=false, max=<0 0>, min=<0 0>}
     }
     disabledForeground {colorUseRule=UseColor_ColorTable, specifiedColor=<rgba 0 0 0 0>}
     hovered {
      bIsDynamicallyLoaded=false
      drawAs=RoundedBox
      imageType=NoImage
      mirroring=NoMirror
      resourceName=None
      resourceObject=null
      tiling=NoTile
      imageSize=<32 32>
      margin=<ltrb 0 0 0 0>
      outlineSettings {
       bUseBrushTransparency=false
       roundingType=FixedRadius
       width=1
       color {colorUseRule=UseColor_ColorTable, specifiedColor=<rgba 0 0 0 0>}
       cornerRadii=<4 4 4 4>
      }
      tintColor {colorUseRule=UseColor_Specified, specifiedColor=<rgba 0.0953 0.0953 0.0953 1>}
      uVRegion {bIsValid=false, max=<0 0>, min=<0 0>}
     }
     hoveredForeground {colorUseRule=UseColor_ColorTable, specifiedColor=<rgba 0 0 0 0>}
     hoveredSlateSound {resourceObject=null}
     normal {
      bIsDynamicallyLoaded=false
      drawAs=RoundedBox
      imageType=NoImage
      mirroring=NoMirror
      resourceName=None
      resourceObject=null
      tiling=NoTile
      imageSize=<32 32>
      margin=<ltrb 0 0 0 0>
      outlineSettings {
       bUseBrushTransparency=false
       roundingType=FixedRadius
       width=1
       color {colorUseRule=UseColor_ColorTable, specifiedColor=<rgba 0 0 0 0>}
       cornerRadii=<4 4 4 4>
      }
      tintColor {colorUseRule=UseColor_Specified, specifiedColor=<rgba 0.0395 0.0395 0.0395 1>}
      uVRegion {bIsValid=false, max=<0 0>, min=<0 0>}
     }
     normalForeground {colorUseRule=UseColor_ColorTable, specifiedColor=<rgba 0 0 0 0>}
     normalPadding=<ltrb 12 1.5 12 1.5>
     pressed {
      bIsDynamicallyLoaded=false
      drawAs=RoundedBox
      imageType=NoImage
      mirroring=NoMirror
      resourceName=None
      resourceObject=null
      tiling=NoTile
      imageSize=<32 32>
      margin=<ltrb 0 0 0 0>
      outlineSettings {
       bUseBrushTransparency=false
       roundingType=FixedRadius
       width=1
       color {colorUseRule=UseColor_ColorTable, specifiedColor=<rgba 0 0 0 0>}
       cornerRadii=<4 4 4 4>
      }
      tintColor {colorUseRule=UseColor_Specified, specifiedColor=<rgba 0.0284 0.0284 0.0284 1>}
      uVRegion {bIsValid=false, max=<0 0>, min=<0 0>}
     }
     pressedForeground {colorUseRule=UseColor_ColorTable, specifiedColor=<rgba 0 0 0 0>}
     pressedPadding=<ltrb 12 2.5 12 0.5>
     pressedSlateSound {resourceObject=null}
    }}
   slot {class=ScaleBoxSlot, properties {horizontalAlignment=HAlign_Fill, verticalAlignment=VAlign_Fill}}
  }
  TextBlock_3 {
   widgetClassPath=&TextBlock
   parent=EditorUtilityButton_1
   widget=TextBlock_3
   properties {
    text=Debug Data to Output Log
    colorAndOpacity {colorUseRule=UseColor_Specified, specifiedColor=<rgba 0.599 0.599 0.599 1>}
    font {
     bForceMonospaced=false
     bMaterialIsStencil=false
     fontMaterial=null
     fontObject=&Roboto
     letterSpacing=0
     monospacedWidth=1
     size=9
     skewAmount=0
     typefaceFontName=Bold
     outlineSettings {
      bApplyOutlineToDropShadows=false
      bMiteredCorners=false
      bSeparateFillAlpha=false
      outlineMaterial=null
      outlineSize=0
      outlineColor=<rgba 0 0 0 1>
     }
    }
   }
   slot {class=ButtonSlot}
  }
 ]
}
