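# Editor utility task that gathers naming and path data from selected assets for the vehicle cosmetics skin tool.
# After a delay and one cancellation check, it loops over the widget's selected objects and appends names, material paths and package paths to the widget's arrays.
# GetNamesData has no logic connected to its outputs, so only the texture package path carries asset data.
# On completion it constructs an unused SkinTools object, then finishes.
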EditorUtilityBlueprint /Game/NeverCook/TechArt/VehicleCosmeticsTools/SkinTools/VehicleCosmetics_SkinTool_GetData
graphs [
 EventGraph {nodes [
   Event BeginExecution()
   > SetTaskNotificationText!(execute, self:Editor Utility Task Object Reference, Text:"Text (by ref)"=MakeLiteralText(Value="LOCTEXT(\"Getting Skins Data\")")) -> (then)
   > PrintString(InString=Getting Selected Asset Data)
   > Delay!(execute, WorldContextObject:Object Reference, Duration=3) -> (then)
   > n1 = Branch(Condition=Get CancelPending() -> (CancelPending:bool))
   n2 = Set SkinToolWidget(
    execute=@n1.else
    SkinToolWidget:Vehicle Cosmetics Skin Tools Object Reference=Get Widget() -> (Widget:Vehicle Cosmetics Skin Tools Object Reference)
   ) -> (Output_Get:Vehicle Cosmetics Skin Tools Object Reference)
   > n3 = For Each Loop with Break!(
    Exec:Exec
    Array:Array of Object References=Get SelectedObjectsArray!(self:Vehicle Cosmetics Skin Tools Object Reference=@n2) -> (SelectedObjectsArray:Array of Object References)
    Break:Exec
   ) -> (LoopBody:Exec, Array Element:Object Reference, Array Index:int, Completed:Exec)
   Construct Object from Class(execute=@n3.Completed, Class=&/VehicleCosmetics_SkinTools.VehicleCosmetics_SkinTools_C) -> (ReturnValue:Vehicle Cosmetics Skin Tools Object Reference)
   > PrintString(InString=Finishing Task..)
   > FinishExecutingTask()
   n4 = CreateAssetData!(InAsset=@n3, bAllowBlueprintClass=false) -> (
    ReturnValue_PackageName:Name
    ReturnValue_PackagePath:Name
    ReturnValue_AssetName:Name
    ReturnValue_AssetClassPath:Top Level Asset Path Structure
   )
   n5 = GetNamesData!(execute=@n3.LoopBody, self, AssetName:Name=@n4.ReturnValue_AssetName, PackagePath:Name=@n4.ReturnValue_PackagePath) -> (then, SkinMI_Name:String, MaterialsPath:String, BodyName:String, BodyMI_Name:String, BodyMI_PackageName:Name=None)
   n6 = Get SkinToolWidget() -> (SkinToolWidget:Vehicle Cosmetics Skin Tools Object Reference)
   > Add(
    TargetArray=Get SkinMatInstanceNames!(self:Vehicle Cosmetics Skin Tools Object Reference=@n6) -> (SkinMatInstanceNames:Array of Strings)
    NewItem=@n5
   ) {is=title K2Node_CallArrayFunction}
   > Add(
    TargetArray=Get BodyMaterialsPath!(self:Vehicle Cosmetics Skin Tools Object Reference=@n6) -> (BodyMaterialsPath:Array of Strings)
    NewItem=@n5.MaterialsPath
   )
   > Add(TargetArray=Get BodyNames!(self:Vehicle Cosmetics Skin Tools Object Reference=@n6) -> (BodyNames:Array of Strings), NewItem=@n5.BodyName)
   > Add(
    TargetArray=Get BodyMatInstanceNames!(self:Vehicle Cosmetics Skin Tools Object Reference=@n6) -> (BodyMatInstanceNames:Array of Strings)
    NewItem=@n5.BodyMI_Name
   )
   > Add(
    TargetArray:Array of Names=Get BodyMatInstancePackageName!(self:Vehicle Cosmetics Skin Tools Object Reference=@n6) -> (BodyMatInstancePackageName:Array of Names)
    NewItem:"Name (by ref)"=@n5.BodyMI_PackageName
   )
   > Add(
    TargetArray:Array of Strings=Get SkinTexturePackagePaths!(self:Vehicle Cosmetics Skin Tools Object Reference=@n6) -> (SkinTexturePackagePaths:Array of Strings)
    NewItem:"String (by ref)"="To String (Name)"(InName=@n4.ReturnValue_PackagePath)
   )
   Event CancelRequested() {is=title K2Node_Event}
   > PrintString(InString=Task Cancelled)
   > Set CancelPending(CancelPending:bool=true) -> (Output_Get:bool)
  ]}
 GetNamesData {nodes [
   GetNamesData!() -> (then, AssetName:Name, PackagePath:Name) {is=title K2Node_FunctionEntry}
   > Return Node!(execute, SkinMI_Name:String, MaterialsPath:String, BodyName:String, BodyMI_Name:String, BodyMI_PackageName:Name=None)
  ]}
]
variables [
 CancelPending {type=boolean}
 NumberOfAssets {type=integer}
 SelectedObjectsArray {type="array<Object>", default []}
 SkinMaskTextures {type="array<Texture>", default []}
 SkinToolWidget {type=VehicleCosmetics_SkinTools_C, default=null}
 Widget {type=VehicleCosmetics_SkinTools_C, default=null}
]
